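# Niagara effect for the Double Duty emote, apparently its initial deploy: a burst of wind, smoke and water.
# Spinning vortex and wind meshes, the latter with lights, combine with radial smoke puffs and a flat smoke sprite.
# User.bIsInWater swaps ground rings for a splash snapped to the water body.
# The rings use scratch trace modules, skippable by platform quality, and align to the collision normal.
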NiagaraSystem /BRCosmetics/Animation/Game/MainPlayer/Emotes/DoubleDuty/FX/NS_DoubleDutEmote_InitalFDeply
dependencies {
 usedDataInterfaces [
  &NiagaraDataInterfaceCurve
  &NiagaraDataInterfaceCollisionQuery
  &NiagaraDataInterfaceDebugDraw
  &NiagaraDataInterfacePlatformSet
  &NiagaraDataInterfaceSpriteRendererInfo
  &NiagaraDataInterfaceVectorCurve
  &NiagaraDataInterfaceColorCurve
  &NiagaraDataInterfaceSkeletalMesh
  &NiagaraDataInterfaceCamera
 ]
 usedDynamicInputs [&ReturnCameraVector]
 usedModules [
  &self:SystemSpawnScript.SystemScriptSource.SystemScriptGraph.NiagaraNodeAssignment_0.SetVariables_#1
  &SystemState
  &self:SystemSpawnScript.SystemScriptSource.SystemScriptGraph.NiagaraNodeAssignment_1.SetVariables_#2
  &self:SystemSpawnScript.SystemScriptSource.SystemScriptGraph.NiagaraNodeAssignment_2.SetVariables_#3
 ]
}
emitters {
 Arm_Vortex {
  emitterUpdateScript [
   &EmitterState()
   &SpawnBurst_Instantaneous(Use Spawn Probability=false, UseLoopCountLimit=false)
  ]
  particleSpawnScript [
   &InitializeParticle(
    Color=<rgba 0.8431 1.0743 1.25 1>
    Lifetime=0.5
    Mesh Scale Max=<1.5 1.5 2>
    Mesh Scale Min=<1.5 1.5 2>
    Mesh Scale Mode=::NewEnumerator2
    UsePositionOffset=false
    Write Material Random=true
   )
   &InitialMeshOrientation(
    Mesh Orientation Mode=::NewEnumerator2
    Rotation=&RandomRangeVector(Minimum=<0 0 0>, Maximum=<0 0 1>)
    Use Rotation Vector=true
   )
   &DynamicMaterialParameters(
    Index 0 Param 1=1
    Index 0 Param 2=-1
    Index 0 Param 3=&RandomRangeFloat(Minimum=-1, Maximum=1)
    Index 0 Param 4=&RandomRangeFloat(Minimum=-1, Maximum=1)
    Module.Index0Param0Write=true
    Module.Index0Param1Write=true
    Module.Index0Param2Write=true
    Module.Index0Param3Write=true
   )
  ]
  particleUpdateScript [
   &ParticleState()
   &UpdateMeshOrientation(Rotation Rate=-1, Rotation Vector=<0 0 1>)
   &DynamicMaterialParameters(
    Index 0 Param 1=1
    Index 1 Param 1=&RandomRangeFloat(Minimum=0, Maximum=1)
    Index 1 Param 2=10
    Index 1 Param 3=50
    Index 1 Param 4=2
    Index 2 Param 1=&MakeFloatFromVector(Vector=$Particles.PositionDifference, Channel=::NewEnumerator0)
    Index 2 Param 2=&MakeFloatFromVector(Vector=$Particles.PositionDifference, Channel=::NewEnumerator1)
    Index 2 Param 3=&MakeFloatFromVector(Vector=$Particles.PositionDifference, Channel=::NewEnumerator2)
    Module.Index0Param0Write=true
    Module.Index0Param1Write=true
    Module.Index0Param2Write=true
    Module.Index0Param3Write=true
    Module.Index1Param0Write=true
    Module.Index1Param1Write=true
    Module.Index1Param2Write=true
    Module.Index1Param3Write=true
    Module.Index2Param0Write=true
    Module.Index2Param1Write=true
    Module.Index2Param2Write=true
    Module.Index2Param3Write=true
    Write Parameter Index 1=true
    Write Parameter Index 2=true
   ) {module=DynamicMaterialParameters001}
   &ScaleColor(
    Scale Alpha=&FloatFromCurve(FloatCurve=DataInterface() {propertyValues {Curve {keys [
         {arriveTangent=9.0725, interpMode=RCIM_Cubic, leaveTangent=9.0725, tangentMode=RCTM_User, time=0, value=0}
         {arriveTangent=-2.5249, interpMode=RCIM_Cubic, leaveTangent=-2.5249, tangentMode=RCTM_User, time=0.1408, value=0.9847}
         {arriveTangent=-0.0405, interpMode=RCIM_Cubic, leaveTangent=-0.0405, tangentMode=RCTM_User, time=1, value=0}
        ]}}})
    ScaleA=true
    ScaleRGB=true
   )
  ]
  rendererClasses [&NiagaraMeshRendererProperties]
  renderers [{rendererClass=&NiagaraMeshRendererProperties, rendererIndex=0}]
 }
 Smoke {
  emitterUpdateScript [
   &EmitterState()
   &SpawnBurst_Instantaneous(Spawn Count=15, Use Spawn Probability=false, UseLoopCountLimit=false)
  ]
  particleSpawnScript [
   &InitializeParticle(
    Color Channel Mode:ENiagaraChannelCorrelation=::NewEnumerator0
    Color Maximum=<rgba 0.7032 0.737 0.8385 0.05>
    Color Minimum=<rgba 0.6328 0.8415 1 0.05>
    Color Mode=::NewEnumerator2
    Lifetime Max=0.8
    Lifetime Min=0.5
    Lifetime Mode=::NewEnumerator1
    Sprite Rotation Mode=::NewEnumerator2
    Sprite Size Mode=::NewEnumerator1
    Uniform Sprite Size Max=100
    Uniform Sprite Size Min=75
    UsePositionOffset=false
    Write Material Random=true
   ) {module=InitializeParticle001}
   &ShapeLocation(
    Cylinder Distribution=::NewEnumerator2
    Cylinder Height=0
    Cylinder Radius=1
    MaskSpawnGroup=false
    Shape Primitive=::NewEnumerator1
    U Position=&RandomRangeFloat(Minimum=0, Maximum=1)
    UseOwnerScale=true
    V Position=&ReturnNormalizedExecIndex()
   )
   &AddVelocity(Velocity=&Multiply_VectorByFloat(Vector=$Particles.ShapeLocation.ShapeNormal, Float=500))
  ]
  particleUpdateScript [
   &ParticleState()
   &Drag(Drag=4, Use Linear Drag=true, Use Rotational Drag=true)
   &SolveForcesAndVelocity(Clamp Velocity=false, Limit Acceleration=false)
   &ScaleSpriteSize(Uniform Curve Sprite Scale=DataInterface() {propertyValues {Curve {keys [
        {arriveTangent=4.1785, interpMode=RCIM_Cubic, leaveTangent=4.1785, tangentMode=RCTM_User, time=0, value=0}
        {arriveTangent=1.1179, interpMode=RCIM_Cubic, leaveTangent=1.1179, tangentMode=RCTM_User, time=0.2042, value=0.6228}
        {arriveTangent=0.0224, interpMode=RCIM_Cubic, leaveTangent=0.0224, tangentMode=RCTM_User, time=1, value=1}
       ]}}})
   &DynamicMaterialParameters(
    Index 0 Param 1=&Add_Float(
     A=&RandomRangeFloat(Minimum=0, Maximum=1)
     B=&FloatFromCurve(
      FloatCurve=DataInterface() {propertyValues {Curve {keys [
          <at 0 0 RCIM_Cubic>
          {arriveTangent=1.5947, interpMode=RCIM_Cubic, leaveTangent=1.5947, tangentMode=RCTM_User, time=0.2557, value=0.5257}
          <at 1 1 RCIM_Cubic>
         ]}}}
      CurveIndex=$Particles.Age
      Scale Curve=0.25
     )
    )
    Module.Param0WriteEnabled=true
    Module.Param1WriteEnabled=true
    Module.Param2WriteEnabled=true
    Module.Param3WriteEnabled=true
   ) {module=DynamicMaterialParameters001}
   &CameraOffset(Camera Offset Amount=15)
   &ScaleColor(
    Scale Alpha=&Multiply_Float(
     A=&FloatFromCurve(
      FloatCurve=DataInterface() {propertyValues {Curve {keys [{interpMode=RCIM_Cubic, tangentMode=RCTM_User, time=0, value=0}, <at 0.15 1 RCIM_Cubic>, <at 0.9 0>]}}}
      Scale Curve=&FN_FrontEndFloatModifier(Gameplay=0.75, Front End=1)
     )
     B=&FloatFromCurve(
      FloatCurve=DataInterface() {propertyValues {Curve {keys [
          {interpMode=RCIM_Cubic, tangentMode=RCTM_User, time=0, value=1}
          {interpMode=RCIM_Cubic, tangentMode=RCTM_User, time=0.5, value=1}
          {tangentMode=RCTM_User, time=1, value=1}
         ]}}}
      CurveIndex=Linked() {linkedVariable {name=NPC.FortniteNPC.TimeOfDay_0-1, type=NiagaraFloat}}
     )
    )
    ScaleA=true
    ScaleRGB=false
   )
  ]
  rendererClasses [&NiagaraSpriteRendererProperties]
  renderers [
   {rendererClass=&NiagaraSpriteRendererProperties}
   {rendererClass=&NiagaraSpriteRendererProperties, rendererIndex=1}
  ]
 }
 SmokeGroundRings {
  emitterUpdateScript [
   &EmitterState(Delay First Loop Only=true, Life Cycle Mode=::NewEnumerator1, Loop Delay=$System.Delay, Loop Duration=1, UseLoopDelay=true)
   &SpawnBurst_Instantaneous(
    Spawn Probability=&MakeCustomFloatFromBool(Bool=$User.bIsInWater, True Float=0, False Float=1)
    Use Spawn Probability=true
    UseLoopCountLimit=false
   )
   &SpawnRate(
    SpawnRate=&Lerp_Float(A=0, B=7, Alpha=&MakeCustomFloatFromBool(Bool=$User.bIsInWater, True Float=0, False Float=1))
    Use Spawn Probability=false
   )
  ]
  particleSpawnScript [
   &InitializeParticle(
    Color Maximum=<rgba 0.7032 0.737 0.8385 1>
    Color Minimum=<rgba 0.2708 0.6852 1 1>
    Color Mode=::NewEnumerator2
    Lifetime=&RandomRangeFloat(Minimum=0.5, Maximum=0.65)
    Mass Mode=::NewEnumerator1
    Position Offset=<0 0 10>
    Sprite Rotation Angle=&UniformRangedFloat(Minimum=-180, Maximum=180)
    Sprite Rotation Mode=::NewEnumerator0
    Sprite Size=&Vector2DFromFloat(Value=5)
    Sprite Size Mode=::NewEnumerator4
    UsePositionOffset=true
    Write Material Random=true
   ) {module=InitializeParticle001}
   &InitialMeshOrientation(Mesh Orientation Mode=::NewEnumerator7, Rotation=<0 0.25 0>, Use Rotation Vector=true)
   &self:ScratchModule(
    Offset End:Vector3f=<0 0 -50>
    Offset Start:Vector3f=<0 0 50>
    SkipTrace:NiagaraBool=&IsPlatformSetActive(New Platform Set=DataInterface() {propertyValues {Platforms {qualityLevelMask=-31}}})
    TracePosition:NiagaraPosition=/Niagara/DynamicInputs/Helpers/SimulationPosition()
   ) {module=ScratchModule001}
  ]
  particleUpdateScript [
   &ParticleState()
   &DynamicMaterialParameters(
    Index 0 Param 1=&FloatFromCurve(FloatCurve=DataInterface() {propertyValues {Curve {keys [<at 0 0.5>,<at 1 2>]}}})
    Module.Param0WriteEnabled=true
    Module.Param1WriteEnabled=false
    Module.Param2WriteEnabled=false
    Module.Param3WriteEnabled=false
   )
   &ScaleSpriteSize(Scale Factor=&Vector2DFromFloat(Value=&FloatFromCurve(
      FloatCurve=DataInterface() {propertyValues {Curve {keys [
          {arriveTangent=35.6763, interpMode=RCIM_Cubic, leaveTangent=3.7835, tangentMode=RCTM_User, time=0, value=0}
          {interpMode=RCIM_Cubic, leaveTangent=0.6903, time=1.0011, value=1}
         ]}}}
      Scale Curve=50
     )))
   &CameraOffset(Camera Offset Amount=20)
   &ScaleColor(
    Scale Alpha=&FloatFromCurve(
     FloatCurve=DataInterface() {propertyValues {Curve {keys [<at 0 1 RCIM_Cubic>, {interpMode=RCIM_Cubic, leaveTangent=-4.9981, time=1.0011, value=0.002}]}}}
     Scale Curve=0.25
    )
    ScaleA=true
    ScaleRGB=false
   )
   &InheritSourceMovement(Applied Position Scale=<0.5 0.5 0.5>, ApplyForce=false, ApplyPosition=true, ApplyPositionFalloff=false)
   &SolveForcesAndVelocity(Clamp Velocity=false, Limit Acceleration=false)
   &self:ScratchModule(
    Offset End:Vector3f=<0 0 -50>
    Offset Start:Vector3f=<0 0 50>
    SkipTrace:NiagaraBool=&IsPlatformSetActive(New Platform Set=DataInterface() {propertyValues {Platforms {qualityLevelMask=-29}}})
    TracePosition:NiagaraPosition=&SimulationPosition()
   ) {module=ScratchModule}
   &UpdateMeshOrientation(
    Facing Direction=$Particles.CollisionNormal
    Facing Interpolation=false
    Orientation Method=::NewEnumerator1
    Override Quat=false
    TransformFacing=false
    TransformUp=false
   )
   &AlignSpriteToMeshOrientation(Mesh Orientation Relative Sprite Facing Vector:Vector3f=<0 0 1>, SetAlignment=true, SetFacing=true)
  ]
  rendererClasses [&NiagaraSpriteRendererProperties]
  renderers [{rendererClass=&NiagaraSpriteRendererProperties}]
 }
 Smoke_Sprite_Flat {
  emitterUpdateScript [
   &EmitterState()
   &SpawnBurst_Instantaneous(Spawn Count=1, Use Spawn Probability=false)
  ]
  particleSpawnScript [
   /Niagara/Modules/Spawn/Initialization/InitializeParticle(
    Color=<rgba 0.3322 0.4431 0.6094 0.75>
    Lifetime=0.6
    Module.Write Color=true
    Module.Write Lifetime=true
    Module.Write Mass=true
    Module.Write Material Random=true
    Module.Write Position=false
    Module.Write Scale=false
    Module.Write SpriteRotation=true
    Module.Write SpriteSize=true
    Position=/Niagara/Functions/Localspace/SimulationPosition()
    Sprite Rotation=&UniformRangedFloat(Minimum=-180, Maximum=180)
    Sprite Size=&Vector2DFromFloat(Value=500)
   ) {enabled=false}
   /Niagara/Modules/Spawn/Initialization/V2/InitializeParticle(
    Color=<rgba 0.3322 0.4431 0.6094 0.5>
    Lifetime=0.6
    Mass Mode=::NewEnumerator1
    Position Offset=<0 0 10>
    Sprite Rotation Angle=&UniformRangedFloat(Minimum=-180, Maximum=180)
    Sprite Rotation Mode=::NewEnumerator0
    Sprite Size=&Vector2DFromFloat(Value=150)
    Sprite Size Mode=::NewEnumerator4
    UsePositionOffset=true
    Write Material Random=true
   ) {module=InitializeParticle001}
   /SetVariables(Particles.SpriteFacing=<0 0 1>) {isSetParameters=true}
   &CameraOffset(Camera Offset Amount=20)
  ]
  particleUpdateScript [
   &ParticleState()
   &DynamicMaterialParameters(
    Index 0 Param 1=&FloatFromCurve(FloatCurve=DataInterface() {propertyValues {Curve {keys [<at 0 0.8>,<at 1 1.5>]}}})
    Module.Param0WriteEnabled=true
    Module.Param1WriteEnabled=false
    Module.Param2WriteEnabled=false
    Module.Param3WriteEnabled=false
   )
   &ScaleSpriteSize(Scale Factor=&Vector2DFromFloat(Value=&FloatFromCurve(
      FloatCurve=DataInterface() {propertyValues {Curve {keys [
          {arriveTangent=4.1785, interpMode=RCIM_Cubic, leaveTangent=4.1785, tangentMode=RCTM_User, time=0, value=0}
          {arriveTangent=1.1179, interpMode=RCIM_Cubic, leaveTangent=1.1179, tangentMode=RCTM_User, time=0.2042, value=0.6228}
          {arriveTangent=0.0224, interpMode=RCIM_Cubic, leaveTangent=0.0224, tangentMode=RCTM_User, time=1, value=1}
         ]}}}
      Scale Curve=3.3
     )))
   &ScaleColor(
    Scale Alpha=&FloatFromCurve(FloatCurve=DataInterface() {propertyValues {Curve {keys [
         <at 0 1 RCIM_Cubic>
         {arriveTangent=-1.4475, interpMode=RCIM_Cubic, leaveTangent=-1.4475, tangentMode=RCTM_User, time=0.8099, value=0.6059}
         {arriveTangent=-4.9981, interpMode=RCIM_Cubic, leaveTangent=-4.9981, tangentMode=RCTM_User, time=1.0011, value=0.002}
        ]}}})
    ScaleA=true
    ScaleRGB=false
   )
  ]
  rendererClasses [&NiagaraSpriteRendererProperties]
  renderers [{rendererClass=&NiagaraSpriteRendererProperties}]
 }
 Splash {
  emitterUpdateScript [
   &EmitterState()
   &SpawnBurst_Instantaneous(
    Spawn Count=&Multiply_Int(
     A=35
     B=&MakeIntFromFloat(Float=&MakeCustomFloatFromBool(Bool=$User.bIsInWater, True Float=1, False Float=0))
    )
    Spawn Time=$System.Delay
    Use Spawn Probability=false
    UseLoopCountLimit=false
   )
  ]
  particleSpawnScript [
   &InitializeParticle(
    Lifetime Max=0.65
    Lifetime Min=0.35
    Lifetime Mode=::NewEnumerator1
    Position Offset=<0 0 10>
    Position Offset Coordinate Space=::Local
    Sprite Rotation Mode=::NewEnumerator2
    Sprite Size=&Multiply_Vector2D_ByFloat(Float=&RandomRangeFloat(Minimum=0.8, Maximum=1), Vector2D=<150 150>)
    Sprite Size Mode=::NewEnumerator4
    UsePositionOffset=true
    Write Material Random=true
   )
   &AddVelocity(
    Cone Angle=180
    Cone Axis=<0 0 1>
    Inner Cone Angle=150
    Module.UseConeFalloff=false
    Module.UseVelDistribution=false
    Override Seed=false
    Velocity Mode=::NewEnumerator2
    Velocity Speed=&RandomRangeFloat(Minimum=200, Maximum=100)
   ) {module=AddVelocity001}
   &AddVelocity(Velocity=<0 0 300>) {module=AddVelocity002}
   /Game/Effects/Niagara/Modules/Water/WaterBodyLocation(
    Snap Position to Water Position=true
    Use Water Plane Position for Water Position=true
    WaterBodyActor:NiagaraDataInterfaceWater=$User.WaterBodyActor
   )
   &SubUVAnimation(
    Evaluation Type=::NewEnumerator0
    "Sprite Renderer "=DataInterface() {propertyValues {SpriteRenderer=&self:Splash_4.NiagaraSpriteRendererProperties_0}}
    SubUV Animation Mode:ENiagara_SubUVLookupModeV2=::NewEnumerator2
    UseEndFrame=false
    UseStartFrame=false
   ) {module=SubUVAnimation001}
  ]
  particleUpdateScript [
   &ParticleState()
   &GravityForce(Gravity=&Multiply_VectorByFloat(Vector=<0 0 -980>, Float=1))
   &InheritSourceMovement(
    Applied Position Scale=&Multiply_VectorByFloat(
     Vector=<1 1 1>
     Float=&FloatFromCurve(FloatCurve=DataInterface() {propertyValues {Curve {keys [{interpMode=RCIM_Cubic, tangentMode=RCTM_User, time=0, value=1}, {interpMode=RCIM_Cubic, tangentMode=RCTM_User, time=1, value=0}]}}})
    )
    ApplyForce=false
    ApplyPosition=true
    ApplyPositionFalloff=false
   )
   &Drag(Drag=2, Use Linear Drag=true, Use Rotational Drag=true)
   &VortexForce(
    UseOriginOffset=false
    Vortex Force Amount=&Multiply_Float(
     A=&RandomRangeFloat(Minimum=1500, Maximum=200)
     B=&UniformAOrBFloat(A=-1, Evaluation Type=::NewEnumerator0, Recalculate Random Each Loop=true, Override Seed=false)
    )
    Vortex Origin:NiagaraPosition=$Particles.SkeletalMeshLocation.SampledPosition
   )
   &SolveForcesAndVelocity(Clamp Velocity=false, Limit Acceleration=false)
   &ScaleSpriteSize(Scale Factor=&Vector2DFromFloat(Value=&FloatFromCurve(FloatCurve=DataInterface() {propertyValues {Curve {keys [<at 0 0.2>,<at 1 1>]}}})))
   &DynamicMaterialParameters(
    Index 0 Param 2=&FloatFromCurve(FloatCurve=DataInterface() {propertyValues {Curve {keys [{interpMode=RCIM_Cubic, tangentMode=RCTM_User, time=0, value=1}, {interpMode=RCIM_Cubic, tangentMode=RCTM_User, time=0.66, value=0}]}}})
    Index 0 Param 4=&FloatFromCurve(FloatCurve=DataInterface() {propertyValues {Curve {keys [{tangentMode=RCTM_User, time=0, value=0}, {tangentMode=RCTM_User, time=1, value=1}]}}})
    Module.Param0WriteEnabled=false
    Module.Param1WriteEnabled=true
    Module.Param2WriteEnabled=false
    Module.Param3WriteEnabled=true
   )
   &Color(Color=&MakeLinearColorFromVectorAndFloat(
     "Vector (RGB)"=<1 1 1>
     "Float (Alpha)"=&FloatFromCurve(FloatCurve=DataInterface() {propertyValues {Curve {keys [<at 0 0 RCIM_Cubic>,<at 0.4 1 RCIM_Cubic>,<at 1 0>]}}})
    ))
  ]
  rendererClasses [&NiagaraSpriteRendererProperties]
  renderers [{rendererClass=&NiagaraSpriteRendererProperties}]
 }
 Wind {
  emitterUpdateScript [
   &EmitterState(Life Cycle Mode=::NewEnumerator1, Loop Behavior=::NewEnumerator1, Loop Duration=1, UseLoopDelay=false)
   &SpawnBurst_Instantaneous(Spawn Count=3, Use Spawn Probability=false, UseLoopCountLimit=false)
   &SpawnRate(SpawnRate=8, Use Spawn Probability=false) {enabled=false}
  ]
  particleSpawnScript [
   /SetVariables(Particles.FlipSide:NiagaraBool=&RandomBool()) {isSetParameters=true}
   &InitializeParticle(
    Color Maximum=<rgba 1.2578 1.3184 1.5 1>
    Color Minimum=<rgba 0.4062 1.0277 1.5 1>
    Color Mode=::NewEnumerator2
    Lifetime=&UniformRangedFloat(Minimum=0.8, Maximum=0.8)
    Mesh Scale=&MakeVector(
     X=&RandomRangeFloat(Minimum=1, Maximum=1.5)
     Y=&RandomRangeFloat(Minimum=1, Maximum=1.5)
     Z=2
    )
    Mesh Scale Mode=::NewEnumerator4
    Position Offset=<0 0 10>
    Position Offset Coordinate Space=::Local
    UsePositionOffset=true
    Write Material Random=true
   ) {module=InitializeParticle001}
   &InitialMeshOrientation(
    Mesh Orientation Mode=::NewEnumerator2
    Rotation=&MakeVector(X=0, Y=0, Z=&RandomRangeFloat(Minimum=-0.1, Maximum=0.1))
    Use Rotation Vector=true
   )
  ]
  particleUpdateScript [
   &ParticleState()
   &UpdateMeshOrientation(
    Rotation Rate=&Multiply_Float(
     A=&MakeCustomFloatFromBool(Bool=$Particles.FlipSide, True Float=-1, False Float=1)
     B=&RandomRangeFloat(Minimum=3, Maximum=4)
    )
    Rotation Vector=<0 0 1>
   )
   &ScaleMeshSize(Scale Factor=&VectorFromCurve(
     VectorCurve=DataInterface() {propertyValues {
       XCurve {keys [
         {arriveTangent=1.1533, interpMode=RCIM_Cubic, leaveTangent=1.1533, tangentMode=RCTM_User, time=0, value=0.5}
         {arriveTangent=0.2322, interpMode=RCIM_Cubic, leaveTangent=0.2322, tangentMode=RCTM_User, time=1, value=1}
        ]}
       YCurve {keys [
         {arriveTangent=1.1533, interpMode=RCIM_Cubic, leaveTangent=1.1533, tangentMode=RCTM_User, time=0, value=0.5}
         {arriveTangent=0.2322, interpMode=RCIM_Cubic, leaveTangent=0.2322, tangentMode=RCTM_User, time=1, value=1}
        ]}
       ZCurve {keys [
         {arriveTangent=1.0009, interpMode=RCIM_Cubic, time=0, value=0.1}
         <at 0.65 1>
         {interpMode=RCIM_Cubic, leaveTangent=0.8008, time=1, value=0.5}
        ]}
      }}
     Scale Curve=&VectorFromFloat(Value=&FloatFromCurve(FloatCurve=DataInterface() {propertyValues {Curve {keys [
           {arriveTangent=4.1785, interpMode=RCIM_Cubic, leaveTangent=4.1785, tangentMode=RCTM_User, time=0, value=0.5}
           {arriveTangent=0.0224, interpMode=RCIM_Cubic, leaveTangent=0.0224, tangentMode=RCTM_User, time=1, value=2}
          ]}}}))
    ))
   &ScaleColor(
    Scale Alpha=&FloatFromCurve(FloatCurve=DataInterface() {propertyValues {Curve {keys [
         {arriveTangent=9.0725, interpMode=RCIM_Cubic, leaveTangent=9.0725, tangentMode=RCTM_User, time=0, value=0}
         {arriveTangent=-2.5249, interpMode=RCIM_Cubic, leaveTangent=-2.5249, tangentMode=RCTM_User, time=0.1408, value=0.9847}
         <at 1 0 RCIM_Cubic>
        ]}}})
    Scale RGB=&MakeVectorFromLinearColorRGB(Color=&ColorFromCurve(ColorCurve=DataInterface() {propertyValues {
        AlphaCurve {keys []}
        BlueCurve {keys [
          {arriveTangent=-19.0486, interpMode=RCIM_Cubic, time=0.2191, value=1.5}
          {arriveTangent=-0.7675, interpMode=RCIM_Cubic, leaveTangent=-0.7675, time=0.6, value=4}
          {interpMode=RCIM_Cubic, leaveTangent=-155.1807, time=0.8706, value=1}
         ]}
        GreenCurve {keys [
          {arriveTangent=-2.2041, interpMode=RCIM_Cubic, time=0.2191, value=0.3817}
          {arriveTangent=-0.0169, interpMode=RCIM_Cubic, leaveTangent=-0.0169, time=0.6, value=0.6495}
          {interpMode=RCIM_Cubic, leaveTangent=-44.728, time=0.8706, value=0.3708}
         ]}
        RedCurve {keys [
          {arriveTangent=-16.8034, interpMode=RCIM_Cubic, time=0.2191, value=0.6539}
          {arriveTangent=-0.263, interpMode=RCIM_Cubic, leaveTangent=-0.263, time=0.6, value=2.9251}
          {interpMode=RCIM_Cubic, leaveTangent=-155.1807, time=0.8706, value=0.4826}
         ]}
       }}))
    ScaleA=true
    ScaleRGB=false
   )
   &DynamicMaterialParameters(
    Index 0 Param 1=&FloatFromCurve(FloatCurve=DataInterface() {propertyValues {Curve {keys [
         {arriveTangent=-9.6411, interpMode=RCIM_Cubic, leaveTangent=-9.6411, tangentMode=RCTM_User, time=0, value=1}
         <at 0.2 0>
         {arriveTangent=1.25, interpMode=RCIM_Cubic, leaveTangent=1.25, time=0.4, value=0}
         {interpMode=RCIM_Cubic, leaveTangent=16.4602, time=1, value=1}
        ]}}})
    Index 0 Param 2=&FloatFromCurve(
     FloatCurve=DataInterface() {propertyValues {Curve {keys [
         {arriveTangent=2.4843, interpMode=RCIM_Cubic, leaveTangent=2.4843, tangentMode=RCTM_User, time=0, value=0}
         {interpMode=RCIM_Cubic, leaveTangent=-0.8461, time=1, value=1}
        ]}}}
     Scale Curve=2
    )
    Index 0 Param 4=&RandomRangeFloat(Minimum=0.4, Maximum=0.6)
    Module.Param0WriteEnabled=true
    Module.Param1WriteEnabled=true
    Module.Param2WriteEnabled=false
    Module.Param3WriteEnabled=true
   )
   &FN_Light_Attributes(
    Color=<rgba 1.0833 2.7407 4 1>
    Create Color Attribute=true
    Create Exponent Attribute=true
    Create Position Attribute=true
    Create Radius Attribute=true
    Exponent=2
    Position=&AddVectorToPosition(Position=$Particles.Position, Vector=<0 0 50>)
    Radius=150
    Scale Color=&FloatFromCurve(FloatCurve=DataInterface() {propertyValues {Curve {keys [
         {arriveTangent=9.0725, interpMode=RCIM_Cubic, leaveTangent=9.0725, tangentMode=RCTM_User, time=0, value=0}
         {arriveTangent=-2.5249, interpMode=RCIM_Cubic, leaveTangent=-2.5249, tangentMode=RCTM_User, time=0.1408, value=0.9847}
         {arriveTangent=-0.0405, interpMode=RCIM_Cubic, leaveTangent=-0.0405, tangentMode=RCTM_User, time=1, value=0}
        ]}}})
   )
  ]
  rendererClasses [&NiagaraMeshRendererProperties, &NiagaraLightRendererProperties]
  renderers [
   {rendererClass=&NiagaraMeshRendererProperties, rendererIndex=0}
   {rendererClass=&NiagaraLightRendererProperties, rendererIndex=1}
  ]
 }
}
summary {
 systemName=NS_DoubleDutEmote_InitalFDeply
 emitters [
  {emitterName=Arm_Vortex, rendererClasses [&NiagaraMeshRendererProperties]}
  {emitterName=SmokeGroundRings, rendererClasses [&NiagaraSpriteRendererProperties]}
  {emitterName=Splash, rendererClasses [&NiagaraSpriteRendererProperties]}
  {emitterName=Wind, rendererClasses [&NiagaraMeshRendererProperties, &NiagaraLightRendererProperties]}
  {emitterName=Smoke_Sprite_Flat, rendererClasses [&NiagaraSpriteRendererProperties]}
  {emitterName=Smoke, rendererClasses [&NiagaraSpriteRendererProperties]}
 ]
 userVariables [
  User.Alpha_G_PointVArray {
   type=NiagaraDataInterfaceArrayFloat3
   description=""
   defaultValue {
    struct=NiagaraExt_VariableValue_DataInterface
    value {
     dataInterface=&self:NiagaraDataInterfaceArrayFloat3_1
     dataInterfaceClass=&NiagaraDataInterfaceArrayFloat3
    }
   }
  }
  User.bIsFrontEnd {type=NiagaraBool}
  User.bIsFrontEndPreview {type=NiagaraBool}
  User.bIsInWater {type=NiagaraBool}
  User.PositionArray {
   type=NiagaraDataInterfaceArrayFloat3
   description=""
   defaultValue {
    struct=NiagaraExt_VariableValue_DataInterface
    value {
     dataInterface=&self:NiagaraDataInterfaceArrayFloat3_0
     dataInterfaceClass=&NiagaraDataInterfaceArrayFloat3
    }
   }
  }
  User.Range_Size {
   type=NiagaraDataInterfaceArrayFloat3
   description=""
   defaultValue {
    struct=NiagaraExt_VariableValue_DataInterface
    value {
     dataInterface=&self:NiagaraDataInterfaceArrayFloat3_3
     dataInterfaceClass=&NiagaraDataInterfaceArrayFloat3
    }
   }
  }
  User.Rate {type=NiagaraFloat, defaultValue=0, description=""}
  User.VelocityDirct {
   type=NiagaraDataInterfaceArrayFloat3
   description=""
   defaultValue {
    struct=NiagaraExt_VariableValue_DataInterface
    value {
     dataInterface=&self:NiagaraDataInterfaceArrayFloat3_2
     dataInterfaceClass=&NiagaraDataInterfaceArrayFloat3
    }
   }
  }
  User.WaterBodyActor {
   type=NiagaraDataInterfaceWater
   description=""
   defaultValue {
    struct=NiagaraExt_VariableValue_DataInterface
    value {
     dataInterface=&self:NiagaraDataInterfaceWater_0
     dataInterfaceClass=/Script/Water.NiagaraDataInterfaceWater
    }
   }
  }
 ]
}
