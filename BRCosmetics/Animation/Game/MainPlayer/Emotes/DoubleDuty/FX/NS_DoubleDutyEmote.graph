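# Niagara system for the DoubleDuty player emote, layering air, ground and water effects on a shared start delay.
# Mesh emitters draw spinning wind, a rotating surface crawl column and a movement-spawned air wake pinned to the owner.
# Ground smoke rings trace to align with the surface and stop spawning when User.bIsInWater is set.
# Splash and foam scale spawning by System.Spawn Water and snap to the supplied water body.
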NiagaraSystem /BRCosmetics/Animation/Game/MainPlayer/Emotes/DoubleDuty/FX/NS_DoubleDutyEmote
dependencies {
 usedDataInterfaces [
  &NiagaraDataInterfaceMeshRendererInfo
  &NiagaraDataInterfaceCurve
  &NiagaraDataInterfaceCollisionQuery
  &NiagaraDataInterfaceDebugDraw
  &NiagaraDataInterfacePlatformSet
  &NiagaraDataInterfaceSpriteRendererInfo
  &NiagaraDataInterfaceVectorCurve
  &NiagaraDataInterfaceColorCurve
  &NiagaraDataInterfaceSkeletalMesh
  &NiagaraDataInterfaceCamera
 ]
 usedDynamicInputs [&ReturnCameraVector]
 usedModules [
  &self:SystemSpawnScript.SystemScriptSource.SystemScriptGraph.NiagaraNodeAssignment_0.SetVariables_#1
  &SystemState
  &self:SystemSpawnScript.SystemScriptSource.SystemScriptGraph.NiagaraNodeAssignment_1.SetVariables_#2
  &self:SystemSpawnScript.SystemScriptSource.SystemScriptGraph.NiagaraNodeAssignment_2.SetVariables_#3
  &self:SystemSpawnScript.SystemScriptSource.SystemScriptGraph.NiagaraNodeAssignment_3.SetVariables_#4
 ]
}
emitters {
 Air_Wake_Back {
  emitterUpdateScript [
   &EmitterState(Delay First Loop Only=true, Life Cycle Mode=::NewEnumerator1, Loop Delay=$System.Delay, Loop Duration=1, UseLoopDelay=true)
   &SpawnRate(SpawnRate=100, Use Spawn Probability=true) {enabled=false}
   &SpawnPerUnit(Module.Use Max Movement Threshold=true, Module.Use Movement Tolerance=false, Spawn Group=1, Spawn Spacing=10, Use Spawn Probability=false)
  ]
  particleSpawnScript [
   /SetVariables(Particles.MeshIndex=&RandomRangeInt(Minimum=0, Maximum=1)) {isSetParameters=true}
   &PartitionParticles()
   &InitializeParticle(
    Color=<rgba 0.7855 0.8961 1 0.5>
    Lifetime Max=1
    Lifetime Min=0.6
    Lifetime Mode=::NewEnumerator1
    Mesh Scale Max=<1 1.25 1.25>
    Mesh Scale Min=<0.75 1 0.25>
    Mesh Scale Mode=::NewEnumerator2
    Position Offset Coordinate Space=::Local
    UsePositionOffset=true
    Write Material Random=true
   )
   &InitialMeshOrientation(
    Mesh Orientation Mode=::NewEnumerator6
    Orientation Quaternion:Quat4f=&FindQuatBetweenTwoVectors(From Vector=<1 0 0>, To Vector=$Engine.Owner.SystemYAxis)
    Rotation=&MakeVector(
     X=&Multiply_Float(
      A=&RandomRangeFloat(Minimum=0, Maximum=0.15)
      B=&Lerp_Float(A=-1, B=1, Alpha=&MakeFloatFromInt(Int=$Particles.PartitionParticles.Partition))
     )
     Y=0
     Z=0
    )
    Use Rotation Vector=true
   )
   &DynamicMaterialParameters(
    Index 0 Param 3=&RandomRangeFloat(Minimum=0, Maximum=0)
    Index 0 Param 4=1.5742
    Module.Index0Param0Write=true
    Module.Index0Param1Write=true
    Module.Index0Param2Write=true
    Module.Index0Param3Write=true
   )
  ]
  particleUpdateScript [
   &ParticleState()
   &UpdateMeshOrientation(
    Rotation Rate=&Multiply_Float(A=0.2, B=&RandomRangeFloat(Minimum=-1, Maximum=1))
    Rotation Vector=<1 0 1>
   ) {enabled=false, module=UpdateMeshOrientation001}
   &SolveForcesAndVelocity(Clamp Velocity=false, Limit Acceleration=false)
   &ScaleColor(
    Scale Alpha=&FloatFromCurve(FloatCurve=DataInterface() {propertyValues {Curve {keys [
         {arriveTangent=9.0725, interpMode=RCIM_Cubic, leaveTangent=9.0725, tangentMode=RCTM_User, time=0, value=0}
         {arriveTangent=-2.5249, interpMode=RCIM_Cubic, leaveTangent=-2.5249, tangentMode=RCTM_User, time=0.1408, value=0.9847}
         {arriveTangent=-0.0405, interpMode=RCIM_Cubic, leaveTangent=-0.0405, tangentMode=RCTM_User, time=1, value=0}
        ]}}})
    ScaleA=true
    ScaleRGB=true
   )
   &UpdateMeshOrientation(
    Rotation Rate=&Lerp_Float(A=-0.1, B=-0.1, Alpha=&MakeFloatFromInt(Int=$Particles.PartitionParticles.Partition))
    Rotation Vector=&RandomRangeVector(Minimum=<0 1 0>, Maximum=<0 2 0>)
   ) {enabled=false}
   &ScaleMeshSize(Scale Factor=&Multiply_VectorByFloat(
     Vector=<1 1 1>
     Float=&FloatFromCurve(FloatCurve=DataInterface() {propertyValues {Curve {keys [
          {arriveTangent=1.2528, interpMode=RCIM_Cubic, leaveTangent=1.2528, tangentMode=RCTM_User, time=0, value=0.25}
          {arriveTangent=0.1691, interpMode=RCIM_Cubic, leaveTangent=0.1691, tangentMode=RCTM_User, time=1, value=1}
         ]}}})
    ))
   /SetVariables(Particles.Position=$Engine.Owner.Position) {isSetParameters=true}
  ]
  rendererClasses [&NiagaraMeshRendererProperties]
  renderers [{rendererClass=&NiagaraMeshRendererProperties, rendererIndex=0}]
 }
 FN_Water_Foam_SnapToSurface {
  emitterUpdateScript [
   &EmitterState(Delay First Loop Only=true, Life Cycle Mode=::NewEnumerator1, Loop Delay=$System.Delay, Loop Duration=1, UseLoopDelay=true)
   &SpawnRate(SpawnRate=&Multiply_Float(A=2.5, B=$System.Spawn Water), Use Spawn Probability=false)
   &SpawnRate(
    SpawnRate=&Multiply_Float(
     A=&Multiply_Float(
      A=20
      B=&VectorLength(Vector=&NormalizeVector(Vector To Normalize=$Engine.Owner.Velocity, Safe Normalize=true, Safe Fallback Threshold=0, Safe Fallback Vector=<0 0 0.001>))
     )
     B=$System.Spawn Water
    )
    Use Spawn Probability=false
   ) {module=SpawnRate001}
  ]
  particleSpawnScript [
   &InitializeParticle(
    Lifetime=&RandomRangeFloat(Minimum=2, Maximum=3)
    Mass Max=1.2
    Mass Min=0.6
    Mass Mode=::NewEnumerator3
    Mesh Scale Mode=::NewEnumerator3
    Mesh Uniform Scale=150
    Position Offset=<0 0 500>
    Sprite Rotation Mode=::NewEnumerator2
    UsePositionOffset=true
    Write Material Random=true
   )
   &InitialMeshOrientation(
    Mesh Orientation Mode=::NewEnumerator2
    Rotation=&RandomRangeVector(Minimum=<0 0 0>, Maximum=<0 0 1>)
    Use Rotation Vector=true
   )
   &InheritVelocity(Module.Limit Speed=true)
  ]
  particleUpdateScript [
   &ParticleState()
   /Game/Effects/Niagara/Modules/Water/WaterBodyLocation(
    Snap Position to Water Position=true
    Use Water Plane Position for Water Position=true
    WaterBodyActor:NiagaraDataInterfaceWater=$User.WaterBodyActor
   )
   &Drag(Drag=2, Use Linear Drag=true, Use Rotational Drag=false)
   &SolveForcesAndVelocity(Clamp Velocity=false, Limit Acceleration=false)
   &DynamicMaterialParameters(
    Index 0 Param 1=&FloatFromCurve(FloatCurve=DataInterface() {propertyValues {Curve {keys [<at -0.0046 1>,<at 1.0011 4>]}}})
    Index 0 Param 2=&FloatFromCurve(FloatCurve=DataInterface() {propertyValues {Curve {keys [<at -0.0046 0>,<at 1.0011 2>]}}})
    Index 0 Param 3=&FloatFromCurve(FloatCurve=DataInterface() {propertyValues {Curve {keys [<at -0.0046 1>,<at 1.0011 0.5>]}}})
    Index 1 Param 3=&Multiply_Float(
     A=200
     B=&Multiply_Float(A=&MakeFloatFromVector(Vector=$Particles.Scale, Channel=::NewEnumerator0), B=0.0067)
    )
    Index 2 Param 1=&FloatFromCurve(FloatCurve=DataInterface() {propertyValues {Curve {keys [<at -0.0046 0.4>,<at 1.0011 1.5>]}}})
    Index 2 Param 2=&FloatFromCurve(FloatCurve=DataInterface() {propertyValues {Curve {keys [<at -0.0046 0.1>,<at 0.11 2.6>,<at 1.0011 0.3>]}}})
    Index 2 Param 3=0.35
    Index 2 Param 4=&FloatFromCurve(FloatCurve=DataInterface() {propertyValues {Curve {keys [
         {arriveTangent=-1.9991, interpMode=RCIM_Cubic, time=0, value=1}
         {arriveTangent=-2.7143, interpMode=RCIM_Cubic, leaveTangent=-2.7143, tangentMode=RCTM_User, time=0.2095, value=0.6901}
         <at 1 0 RCIM_Cubic>
        ]}}})
    Module.Index0Param0Write=true
    Module.Index0Param1Write=true
    Module.Index0Param2Write=true
    Module.Index0Param3Write=false
    Module.Index1Param0Write=false
    Module.Index1Param1Write=false
    Module.Index1Param2Write=true
    Module.Index1Param3Write=false
    Module.Index2Param0Write=true
    Module.Index2Param1Write=true
    Module.Index2Param2Write=true
    Module.Index2Param3Write=true
    Write Parameter Index 1=true
    Write Parameter Index 2=true
   )
   &ScaleColor(ScaleA=true, ScaleRGB=true)
  ]
  rendererClasses [&NiagaraMeshRendererProperties]
  renderers [{rendererClass=&NiagaraMeshRendererProperties, rendererIndex=0}]
 }
 SmokeGroundRings {
  emitterUpdateScript [
   &EmitterState(Delay First Loop Only=true, Life Cycle Mode=::NewEnumerator1, Loop Delay=$System.Delay, Loop Duration=1, UseLoopDelay=true)
   &SpawnRate(
    SpawnRate=&Lerp_Float(A=0, B=4, Alpha=&MakeCustomFloatFromBool(Bool=$User.bIsInWater, True Float=0, False Float=1))
    Use Spawn Probability=false
   )
  ]
  particleSpawnScript [
   /Niagara/Modules/Spawn/Initialization/InitializeParticle(
    Color=<rgba 0.35 0.2778 0.21 0.7>
    Lifetime=&RandomRangeFloat(Minimum=0.95, Maximum=1.1)
    Module.Write Color=true
    Module.Write Lifetime=true
    Module.Write Mass=true
    Module.Write Material Random=true
    Module.Write Position=false
    Module.Write Scale=false
    Module.Write SpriteRotation=true
    Module.Write SpriteSize=true
    Position=&SimulationPosition()
    Sprite Rotation=&UniformRangedFloat(Minimum=-180, Maximum=180)
    Sprite Size=&Vector2DFromFloat(Value=5)
   ) {enabled=false}
   /Niagara/Modules/Spawn/Initialization/V2/InitializeParticle(
    Color=&RandomRangeLinearColor(Minimum=<rgba 0.2708 0.6852 1 1>, Maximum=<rgba 0.7032 0.737 0.8385 1>)
    Lifetime=&RandomRangeFloat(Minimum=0.95, Maximum=1.1)
    Mass Mode=::NewEnumerator1
    Sprite Rotation Angle=&UniformRangedFloat(Minimum=-180, Maximum=180)
    Sprite Rotation Mode=::NewEnumerator0
    Sprite Size=&Vector2DFromFloat(Value=5)
    Sprite Size Mode=::NewEnumerator4
    UsePositionOffset=false
    Write Material Random=true
   ) {module=InitializeParticle001}
   /SetVariables(Particles.SpriteFacing=<0 0 1>) {enabled=false, isSetParameters=true}
   &self:ScratchModule(
    Offset End:Vector3f=<0 0 -250>
    Offset Start:Vector3f=<0 0 250>
    SkipTrace:NiagaraBool=&IsPlatformSetActive(New Platform Set=DataInterface() {propertyValues {Platforms {qualityLevelMask=-29}}})
    TracePosition:NiagaraPosition=$Particles.Initial.SkeletalMeshLocation.SampledPosition
   ) {enabled=false, module=ScratchModule001}
   &InitialMeshOrientation(Mesh Orientation Mode=::NewEnumerator7, Use Rotation Vector=true)
   &self:ScratchModule(
    Offset End:Vector3f=<0 0 -50>
    Offset Start:Vector3f=<0 0 50>
    TracePosition:NiagaraPosition=&AddVectorToPosition(Position=/Niagara/DynamicInputs/Helpers/SimulationPosition(), Vector=<0 0 100>)
   ) {module=ScratchModule002}
  ]
  particleUpdateScript [
   &ParticleState()
   &DynamicMaterialParameters(
    Index 0 Param 1=&FloatFromCurve(FloatCurve=DataInterface() {propertyValues {Curve {keys [<at 0 0.5>,<at 1 2>]}}})
    Module.Param0WriteEnabled=true
    Module.Param1WriteEnabled=false
    Module.Param2WriteEnabled=false
    Module.Param3WriteEnabled=false
   )
   &ScaleSpriteSize(Scale Factor=&Vector2DFromFloat(Value=&FloatFromCurve(
      FloatCurve=DataInterface() {propertyValues {Curve {keys [
          {arriveTangent=35.6763, interpMode=RCIM_Cubic, leaveTangent=3.7835, tangentMode=RCTM_User, time=0, value=0}
          {interpMode=RCIM_Cubic, leaveTangent=0.6903, time=1.0011, value=1}
         ]}}}
      Scale Curve=40
     )))
   &CameraOffset(Camera Offset Amount=20)
   &ScaleColor(
    Scale Alpha=&FloatFromCurve(
     FloatCurve=DataInterface() {propertyValues {Curve {keys [<at 0 1 RCIM_Cubic>, {interpMode=RCIM_Cubic, leaveTangent=-4.9981, time=1.0011, value=0.002}]}}}
     Scale Curve=0.5
    )
    ScaleA=true
    ScaleRGB=false
   )
   &SolveForcesAndVelocity(Clamp Velocity=false, Limit Acceleration=false)
   &ScaleColor(
    Scale Alpha=&Lerp_Float(
     A=1
     B=0.35
     Alpha=&FloatFromCurve(
      FloatCurve=DataInterface() {propertyValues {Curve {keys [<at 1 0>,<at 200 1>]}}}
      CurveIndex=&VectorLength(Vector=$Engine.Owner.Velocity)
     )
    )
    ScaleA=true
    ScaleRGB=true
   ) {module=ScaleColor001}
   &self:ScratchModule(
    Offset End:Vector3f=<0 0 -250>
    Offset Start:Vector3f=<0 0 250>
    SkipTrace:NiagaraBool=&IsPlatformSetActive(New Platform Set=DataInterface() {propertyValues {Platforms {qualityLevelMask=-29}}})
    TracePosition:NiagaraPosition=&AddVectorToPosition(Position=&SimulationPosition(), Vector=<0 0 100>)
   ) {module=ScratchModule}
   &UpdateMeshOrientation(
    Facing Direction=$Particles.CollisionNormal
    Facing Interpolation=false
    Orientation Method=::NewEnumerator1
    Override Quat=false
    TransformFacing=false
    TransformUp=false
   )
   &AlignSpriteToMeshOrientation(Mesh Orientation Relative Sprite Facing Vector:Vector3f=<0 0 1>, SetAlignment=true, SetFacing=true)
  ]
  rendererClasses [&NiagaraSpriteRendererProperties]
  renderers [{rendererClass=&NiagaraSpriteRendererProperties}]
 }
 Splash {
  emitterUpdateScript [
   &EmitterState(Delay First Loop Only=true, Life Cycle Mode=::NewEnumerator1, Loop Delay=$System.Delay, Loop Duration=0, UseLoopDelay=true)
   &SpawnRate(SpawnRate=&Multiply_Float(A=20, B=$System.Spawn Water), Use Spawn Probability=false)
  ]
  particleSpawnScript [
   &InitializeParticle(
    Lifetime Max=0.65
    Lifetime Min=0.35
    Lifetime Mode=::NewEnumerator1
    Position Offset=&MakeVector(
     X=&RandomRangeFloat(Minimum=-20, Maximum=-50)
     Y=&RandomRangeFloat(Minimum=-30, Maximum=30)
     Z=-10
    )
    Position Offset Coordinate Space=::Local
    Sprite Rotation Mode=::NewEnumerator2
    Sprite Size=&Multiply_Vector2D_ByFloat(Float=&RandomRangeFloat(Minimum=0.8, Maximum=1), Vector2D=<150 150>)
    Sprite Size Mode=::NewEnumerator4
    UsePositionOffset=true
    Write Material Random=true
   )
   &SkeletalMeshLocation(Bone Sampling Mode=::NewEnumerator4, MaskSpawnGroup=false, Orientation Sampling=::NewEnumerator1, Skeletal Mesh=$System.Skeletal Mesh)
   &WaterBodyLocation(
    Snap Position to Water Position=true
    Use Water Plane Position for Water Position=true
    WaterBodyActor:NiagaraDataInterfaceWater=$User.WaterBodyActor
   )
   &AddVelocity(
    Cone Angle=180
    Cone Axis=<0 0 1>
    Inner Cone Angle=150
    Module.UseConeFalloff=false
    Module.UseVelDistribution=false
    Override Seed=false
    Velocity Mode=::NewEnumerator2
    Velocity Speed=&RandomRangeFloat(Minimum=200, Maximum=100)
   ) {module=AddVelocity001}
   &AddVelocity(Velocity=<0 0 300>) {module=AddVelocity002}
   &SubUVAnimation(
    Evaluation Type=::NewEnumerator0
    "Sprite Renderer "=DataInterface() {propertyValues {SpriteRenderer=&self:Splash_4.NiagaraSpriteRendererProperties_0}}
    SubUV Animation Mode:ENiagara_SubUVLookupModeV2=::NewEnumerator2
    UseEndFrame=false
    UseStartFrame=false
   ) {module=SubUVAnimation001}
  ]
  particleUpdateScript [
   &ParticleState()
   &GravityForce(Gravity=&Multiply_VectorByFloat(Vector=<0 0 -980>, Float=1))
   &InheritSourceMovement(
    Applied Position Scale=&Multiply_VectorByFloat(
     Vector=<1 1 1>
     Float=&FloatFromCurve(FloatCurve=DataInterface() {propertyValues {Curve {keys [{interpMode=RCIM_Cubic, tangentMode=RCTM_User, time=0, value=1}, {interpMode=RCIM_Cubic, tangentMode=RCTM_User, time=1, value=0}]}}})
    )
    ApplyForce=false
    ApplyPosition=true
    ApplyPositionFalloff=false
   )
   &Drag(Drag=2, Use Linear Drag=true, Use Rotational Drag=true)
   &VortexForce(
    UseOriginOffset=false
    Vortex Force Amount=&Multiply_Float(
     A=&RandomRangeFloat(Minimum=1500, Maximum=200)
     B=&UniformAOrBFloat(A=-1, Evaluation Type=::NewEnumerator0, Recalculate Random Each Loop=true, Override Seed=false)
    )
    Vortex Origin:NiagaraPosition=$Particles.SkeletalMeshLocation.SampledPosition
   )
   &SolveForcesAndVelocity(Clamp Velocity=false, Limit Acceleration=false)
   &ScaleSpriteSize(Scale Factor=&Vector2DFromFloat(Value=&FloatFromCurve(FloatCurve=DataInterface() {propertyValues {Curve {keys [<at 0 0.2>,<at 1 1>]}}})))
   &DynamicMaterialParameters(
    Index 0 Param 2=&FloatFromCurve(FloatCurve=DataInterface() {propertyValues {Curve {keys [{interpMode=RCIM_Cubic, tangentMode=RCTM_User, time=0, value=1}, {interpMode=RCIM_Cubic, tangentMode=RCTM_User, time=0.66, value=0}]}}})
    Index 0 Param 4=&FloatFromCurve(FloatCurve=DataInterface() {propertyValues {Curve {keys [{tangentMode=RCTM_User, time=0, value=0}, {tangentMode=RCTM_User, time=1, value=1}]}}})
    Module.Param0WriteEnabled=false
    Module.Param1WriteEnabled=true
    Module.Param2WriteEnabled=false
    Module.Param3WriteEnabled=true
   )
   &Color(Color=&MakeLinearColorFromVectorAndFloat(
     "Vector (RGB)"=<1 1 1>
     "Float (Alpha)"=&FloatFromCurve(FloatCurve=DataInterface() {propertyValues {Curve {keys [<at 0 0 RCIM_Cubic>,<at 0.4 1 RCIM_Cubic>,<at 1 0>]}}})
    ))
  ]
  rendererClasses [&NiagaraSpriteRendererProperties]
  renderers [{rendererClass=&NiagaraSpriteRendererProperties}]
 }
 SurfaceCrawl {
  emitterUpdateScript [
   &EmitterState()
   &SpawnBurst_Instantaneous(Spawn Time=$System.Delay, Use Spawn Probability=false, UseLoopCountLimit=false)
   &SpawnRate(SpawnRate=3, Use Spawn Probability=false)
  ]
  particleSpawnScript [
   /SetVariables(Particles.MeshOrientation=&MakeQuatFromAxisAngle(Axis=&RandomRangeVector(Minimum=<2 -1.4 2>, Maximum=<2 1.4 2>), AngleInDegrees=90)) {enabled=false, isSetParameters=true}
   &InitializeParticle(
    Color Maximum=<rgba 0.5 0.7841 1 0.7>
    Color Minimum=<rgba 0.7032 0.737 0.8385 0.5>
    Color Mode=::NewEnumerator2
    Lifetime Mode=::NewEnumerator1
    Material Random:NiagaraFloat=&RandomRangeFloat(Minimum=0.05, Maximum=0.95)
    Mesh Renderer Array Visibility Mode=::NewEnumerator3
    Mesh Renderer Info=DataInterface() {propertyValues {MeshRenderer=&self:SurfaceCrawl_1.NiagaraMeshRendererProperties_0}}
    Mesh Scale=&Multiply_VectorByFloat(Vector=&RandomRangeVector(Minimum=<0.11 0.11 0.11>, Maximum=<0.1 0.1 0.1>), Float=1)
    Mesh Scale Mode=::NewEnumerator4
    UsePositionOffset=false
    Write Material Random=true
   ) {module=InitializeParticle001}
   /SetVariables(Particles.SphereHeight_0-1:NiagaraFloat=&RandomRangeFloat(Minimum=0, Maximum=1)) {isSetParameters=true}
   &CameraOffset(Camera Offset Amount=&RandomRangeFloat(Minimum=3, Maximum=4)) {enabled=false}
   &DynamicMaterialParameters(
    Index 0 Param 1=1
    Index 0 Param 2=1
    Index 0 Param 3=1
    Index 0 Param 4=&RandomRangeFloat(Minimum=0.5, Maximum=1)
    Module.Param0WriteEnabled=true
    Module.Param1WriteEnabled=true
    Module.Param2WriteEnabled=true
    Module.Param3WriteEnabled=true
   ) {module=DynamicMaterialParameters001}
   &InitialMeshOrientation(
    Mesh Orientation Mode=::NewEnumerator7
    Rotation=&Multiply_Vector(A=&RandomRangeVector(Minimum=<0.075 0.075 0>, Maximum=<-0.075 -0.075 1>), B=<0.1 0.1 1>)
    Use Rotation Vector=true
   )
   &ShapeLocation(
    Cylinder Distribution=::NewEnumerator2
    Cylinder Height=60
    Cylinder Radius=0
    MaskSpawnGroup=false
    Offset=<0 0 34>
    Offset Mode=::NewEnumerator0
    Shape Primitive=::NewEnumerator1
    U Position=&RandomRangeFloat(Minimum=0, Maximum=1)
    UseOwnerScale=true
    V Position=Linked() {linkedVariable {name=Particles.SphereHeight_0-1, type=NiagaraFloat}}
   )
  ]
  particleUpdateScript [
   &ParticleState()
   &ScaleMeshSize(Scale Factor=&VectorFromFloat(Value=&FloatFromCurve(
      FloatCurve=DataInterface() {propertyValues {Curve {keys [
          {arriveTangent=1.9537, interpMode=RCIM_Cubic, leaveTangent=1.9537, tangentMode=RCTM_User, time=0, value=0.5}
          {interpMode=RCIM_Cubic, tangentMode=RCTM_User, time=0.5, value=1}
          {arriveTangent=-1.4081, tangentMode=RCTM_User, time=1, value=0.5}
         ]}}}
      CurveIndex=Linked() {linkedVariable {name=Particles.SphereHeight_0-1, type=NiagaraFloat}}
      Scale Curve=8
     ))) {module=ScaleMeshSize002}
   &UpdateMeshOrientation(
    Rotation Rate=&Multiply_Float(
     A=&RandomRangeFloat(Minimum=2, Maximum=1)
     B=&UniformAOrBFloat(A=-1, Evaluation Type=::NewEnumerator0, Recalculate Random Each Loop=true, Override Seed=false)
    )
    Rotation Vector=<0 0 1>
   )
   &ScaleMeshSize(Scale Factor=&VectorFromFloat(Value=&FloatFromCurve(
      FloatCurve=DataInterface() {propertyValues {Curve {keys [
          {arriveTangent=5.6263, interpMode=RCIM_Cubic, leaveTangent=5.6263, tangentMode=RCTM_User, time=0, value=0}
          <at 0.5 1 RCIM_Cubic>
          {arriveTangent=-5.6263, interpMode=RCIM_Cubic, leaveTangent=-5.6263, tangentMode=RCTM_User, time=1, value=0}
         ]}}}
      Scale Curve=5.835
     ))) {enabled=false}
   &DynamicMaterialParameters(
    Index 0 Param 1=1
    Index 0 Param 2=1
    Index 0 Param 3=1
    Index 0 Param 4=&Multiply_Float(
     A=&RandomRangeFloat(Minimum=0.75, Maximum=0.8)
     B=&FloatFromCurve(
      FloatCurve=DataInterface() {propertyValues {Curve {keys [
          {interpMode=RCIM_Cubic, tangentMode=RCTM_User, time=0, value=0.75}
          {interpMode=RCIM_Cubic, tangentMode=RCTM_User, time=0.5, value=1}
          {tangentMode=RCTM_User, time=1, value=0.75}
         ]}}}
      CurveIndex=$Particles.MaterialRandom
     )
    )
    Module.Param0WriteEnabled=true
    Module.Param1WriteEnabled=true
    Module.Param2WriteEnabled=true
    Module.Param3WriteEnabled=true
   )
   &UpdateMeshOrientation(Rotation Rate=&RandomRangeFloat(Minimum=0.5, Maximum=-0.5), Rotation Vector=<0 0 1>) {enabled=false, module=UpdateMeshOrientation001}
   &ScaleMeshSize(Scale Factor=&self:ArcTanScale(
     Scale:NiagaraFloat=&Divide_Float(
      A=&DistanceBetweenPositions(A=$Particles.SkeletalMeshLocation.SampledPosition, B=$Particles.SkeletalMeshLocation001.SampledPosition)
      B=&RandomRangeFloat(Minimum=8, Maximum=9)
     )
     Time:NiagaraFloat=$Particles.MaterialRandom
    )) {enabled=false, module=ScaleMeshSize001}
   &ScaleColor(
    Scale Alpha=&FloatFromCurve(
     FloatCurve=DataInterface() {propertyValues {Curve {keys [<at 0 0.75 RCIM_Cubic>,<at 1 0.25 RCIM_Cubic>]}}}
     CurveIndex=&Abs_Float(Float:NiagaraFloat=&DotProduct(Vector A=$System.CamForVecWS, Vector B=<0 0 1>))
    )
    ScaleA=true
    ScaleRGB=true
   )
   &ScaleColor(
    Scale Alpha=&FloatFromCurve(
     FloatCurve=DataInterface() {propertyValues {Curve {keys [
         {
          arriveTangent=4.2519
          interpMode=RCIM_Cubic
          leaveTangent=4.2519
          leaveTangentWeight=0.728
          tangentMode=RCTM_User
          tangentWeightMode=RCTWM_WeightedBoth
          time=0
          value=0
         }
         {
          arriveTangentWeight=0.3498
          interpMode=RCIM_Cubic
          leaveTangentWeight=0.3498
          tangentMode=RCTM_User
          tangentWeightMode=RCTWM_WeightedBoth
          time=0.5
          value=1
         }
         {arriveTangent=-4.3961, interpMode=RCIM_Cubic, leaveTangent=-4.3961, tangentMode=RCTM_User, time=1, value=0}
        ]}}}
     Scale Curve=2
    )
    ScaleA=true
    ScaleRGB=true
   ) {module=ScaleColor001}
  ]
  rendererClasses [&NiagaraMeshRendererProperties]
  renderers [{rendererClass=&NiagaraMeshRendererProperties, rendererIndex=0}]
 }
 Wind {
  emitterUpdateScript [
   &EmitterState(
    Life Cycle Mode=::NewEnumerator1
    Loop Behavior=::NewEnumerator1
    Loop Delay=$System.Delay
    Loop Duration Mode=::NewEnumerator1
    UseLoopDelay=true
   )
   &SpawnBurst_Instantaneous(Use Spawn Probability=false, UseLoopCountLimit=false)
   &SpawnRate(SpawnRate=16, Use Spawn Probability=false)
  ]
  particleSpawnScript [
   /SetVariables(Particles.FlipSide:NiagaraBool=&RandomBool()) {isSetParameters=true}
   &InitializeParticle(
    Color Maximum=<rgba 0.7032 0.737 0.8385 1>
    Color Minimum=<rgba 0.2708 0.6852 1 1>
    Color Mode=::NewEnumerator2
    Lifetime=&UniformRangedFloat(Minimum=0.35, Maximum=0.5)
    Mesh Scale=&MakeVector(
     X=&RandomRangeFloat(Minimum=0.75, Maximum=1)
     Y=&RandomRangeFloat(Minimum=0.75, Maximum=1)
     Z=2
    )
    Mesh Scale Mode=::NewEnumerator4
    Position Offset=<0 0 15>
    Position Offset Coordinate Space=::Local
    UsePositionOffset=true
    Write Material Random=true
   ) {module=InitializeParticle001}
   &InitialMeshOrientation(
    Mesh Orientation Mode=::NewEnumerator2
    Rotation=&MakeVector(X=0, Y=0, Z=&RandomRangeFloat(Minimum=-0.1, Maximum=0.1))
    Use Rotation Vector=true
   )
   &CameraOffset()
  ]
  particleUpdateScript [
   &ParticleState()
   &UpdateMeshOrientation(
    Rotation Rate=&Multiply_Float(
     A=&MakeCustomFloatFromBool(Bool=$Particles.FlipSide, True Float=-1, False Float=1)
     B=&RandomRangeFloat(Minimum=3, Maximum=4)
    )
    Rotation Vector=<0 0 1>
   )
   &ScaleMeshSize(Scale Factor=&VectorFromCurve(VectorCurve=DataInterface() {propertyValues {
       XCurve {keys [
         {arriveTangent=1.1533, interpMode=RCIM_Cubic, leaveTangent=1.1533, tangentMode=RCTM_User, time=0, value=0.5}
         {arriveTangent=0.2322, interpMode=RCIM_Cubic, leaveTangent=0.2322, tangentMode=RCTM_User, time=1, value=1}
        ]}
       YCurve {keys [
         {arriveTangent=1.1533, interpMode=RCIM_Cubic, leaveTangent=1.1533, tangentMode=RCTM_User, time=0, value=0.5}
         {arriveTangent=0.2322, interpMode=RCIM_Cubic, leaveTangent=0.2322, tangentMode=RCTM_User, time=1, value=1}
        ]}
       ZCurve {keys [
         {arriveTangent=1.0009, interpMode=RCIM_Cubic, time=0, value=0.1}
         <at 0.65 1>
         {interpMode=RCIM_Cubic, leaveTangent=0.8008, time=1, value=0.5}
        ]}
      }}))
   &ScaleColor(
    Scale Alpha=&FloatFromCurve(
     FloatCurve=DataInterface() {propertyValues {Curve {keys [
         {arriveTangent=8.125, interpMode=RCIM_Cubic, time=0, value=0}
         {arriveTangentWeight=0.4805, interpMode=RCIM_Cubic, leaveTangentWeight=0.1363, tangentMode=RCTM_User, time=0.6, value=1}
         {interpMode=RCIM_Cubic, leaveTangent=-2.3395, time=1, value=0}
        ]}}}
     Scale Curve=&FNTimeOfDayScalar(Night Scale=0.5, Morning Scale=0.5, Day Scale=0.25, Evening Scale=0.5)
    )
    Scale RGB=&MakeVectorFromLinearColorRGB(Color=&ColorFromCurve(ColorCurve=DataInterface() {propertyValues {
        AlphaCurve {keys []}
        BlueCurve {keys [
          {arriveTangent=-19.0486, interpMode=RCIM_Cubic, time=0.2191, value=1.5}
          {arriveTangent=-0.7675, interpMode=RCIM_Cubic, leaveTangent=-0.7675, time=0.6, value=4}
          {interpMode=RCIM_Cubic, leaveTangent=-155.1807, time=0.8706, value=1}
         ]}
        GreenCurve {keys [
          {arriveTangent=-2.2041, interpMode=RCIM_Cubic, time=0.2191, value=0.3817}
          {arriveTangent=-0.0169, interpMode=RCIM_Cubic, leaveTangent=-0.0169, time=0.6, value=0.6495}
          {interpMode=RCIM_Cubic, leaveTangent=-44.728, time=0.8706, value=0.3708}
         ]}
        RedCurve {keys [
          {arriveTangent=-16.8034, interpMode=RCIM_Cubic, time=0.2191, value=0.6539}
          {arriveTangent=-0.263, interpMode=RCIM_Cubic, leaveTangent=-0.263, time=0.6, value=2.9251}
          {interpMode=RCIM_Cubic, leaveTangent=-155.1807, time=0.8706, value=0.4826}
         ]}
       }}))
    ScaleA=true
    ScaleRGB=false
   )
   &DynamicMaterialParameters(
    Index 0 Param 1=&FloatFromCurve(FloatCurve=DataInterface() {propertyValues {Curve {keys [
         {arriveTangent=-9.6411, interpMode=RCIM_Cubic, leaveTangent=-9.6411, tangentMode=RCTM_User, time=0, value=1}
         <at 0.2 0>
         {arriveTangent=1.25, interpMode=RCIM_Cubic, leaveTangent=1.25, time=0.4, value=0}
         {interpMode=RCIM_Cubic, leaveTangent=16.4602, time=1, value=1}
        ]}}})
    Index 0 Param 2=&FloatFromCurve(
     FloatCurve=DataInterface() {propertyValues {Curve {keys [
         {arriveTangent=2.4843, interpMode=RCIM_Cubic, leaveTangent=2.4843, tangentMode=RCTM_User, time=0, value=0}
         {interpMode=RCIM_Cubic, leaveTangent=-0.8461, time=1, value=1}
        ]}}}
     Scale Curve=2
    )
    Index 0 Param 4=&RandomRangeFloat(Minimum=0.4, Maximum=0.6)
    Module.Param0WriteEnabled=true
    Module.Param1WriteEnabled=true
    Module.Param2WriteEnabled=false
    Module.Param3WriteEnabled=true
   )
  ]
  rendererClasses [&NiagaraMeshRendererProperties]
  renderers [{rendererClass=&NiagaraMeshRendererProperties, rendererIndex=0}]
 }
}
summary {
 systemName=NS_DoubleDutyEmote
 emitters [
  {emitterName=SurfaceCrawl, rendererClasses [&NiagaraMeshRendererProperties]}
  {emitterName=SmokeGroundRings, rendererClasses [&NiagaraSpriteRendererProperties]}
  {emitterName=Air_Wake_Back, rendererClasses [&NiagaraMeshRendererProperties]}
  {emitterName=Splash, rendererClasses [&NiagaraSpriteRendererProperties]}
  {emitterName=FN_Water_Foam_SnapToSurface, rendererClasses [&NiagaraMeshRendererProperties]}
  {emitterName=Wind, rendererClasses [&NiagaraMeshRendererProperties]}
 ]
 userVariables [
  User.bIsFrontEnd {type=NiagaraBool}
  User.bIsFrontEndPreview {type=NiagaraBool}
  User.bIsInWater {type=NiagaraBool}
  User.Boost {type=NiagaraFloat, defaultValue=0, description=""}
  User.ForwardSpeed {type=NiagaraFloat, defaultValue=0, description=""}
  User.IsInWater {type=NiagaraBool}
  User.PhysicalSurface {type=NiagaraInt32}
  User.WaterBodyActor {
   type=NiagaraDataInterfaceWater
   description=""
   defaultValue {
    struct=NiagaraExt_VariableValue_DataInterface
    value {
     dataInterface=&self:NiagaraDataInterfaceWater_0
     dataInterfaceClass=/Script/Water.NiagaraDataInterfaceWater
    }
   }
  }
 ]
}
